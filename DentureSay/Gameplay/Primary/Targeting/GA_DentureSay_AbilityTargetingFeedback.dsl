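; Targeting feedback ability for the DentureSay primary that only performs targeting while the avatar pawn is falling.
; Overrides PerformTargeting, calling the parent's logic when character movement reports falling.
; Otherwise it reschedules itself with a one-shot timer at the hotfixable valid-targeting check rate, or on the next tick if that rate is not positive.
; An invalid avatar pawn ends the event without rescheduling.
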(event EventPerformTargeting
  (bind _returnvalue (CallFunction|GetAvatarasFortPlayerPawn))
  (bind _hotfix_validtargetingcheckrate (Variables|Hotfix|GetHotfix_ValidTargetingCheckRate))
  (bind _returnvalue_2 (ScalableFloat|GetValueatLevel _hotfix_validtargetingcheckrate))
  (bind _outputdelegate (EventDispatchers|CreateEvent :!event PerformTargeting))
  ; Only perform targeting while falling
  (Utilities|IsValid _returnvalue
    (:Is%20Valid
      (bind _charactermovement (Variables|/Script/Engine.Character|GetCharacterMovement _returnvalue)
        (:then
          (if (AI|Components|NavMovement|IsFalling _charactermovement)
            (|CallParentFunction)
            (elif (> _returnvalue_2 0)
              (Utilities|Time|SetTimerbyEvent _outputdelegate _returnvalue_2 false true)
              (else
                (Utilities|Time|SetTimerforNextTickbyEvent _outputdelegate)))))
        (:else
          _)))
    (:Is%20Not%20Valid
      _)))
